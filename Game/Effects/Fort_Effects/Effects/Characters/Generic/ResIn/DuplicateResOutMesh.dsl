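; Short-lived actor that plays the teleport/res in or out dissolve on a pawn or weapon by duplicating its skeletal meshes.
; Duplicates follow the source via leader pose and per-tick transform copy, wearing RezIn materials while source materials become masked MIDs.
; A timeline, forward for out and reversed for in, drives a socket-aligned dissolve gradient and optional non-mobile gun light.
; Finishing in reverse restores the original materials.
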(event EventTick
  (params (DeltaSeconds :type Float%20%28single-precision%29))
  ;[1 0.05157 0.08141] Attaching one object to another didn't work in practice but stealing the external objects transform on tick did. The actor only exists for a few seconds so it should be fine .
  (for (_array_element _array_index) (Variables|Default|GetExternalSkeletalMeshComponent)
    (Transformation|SetWorldTransform (Utilities|Array|Get(acopy) (Variables|Default|GetInternalSkeletalMeshComponents) _array_index) (Transformation|GetWorldTransform _array_element) false false)))

(event Custom|UpdateExternalParameters
  (params (MID%20Array :type Array%20of%20Material%20Instance%20Dynamic%20Object%20References) (External%20MEsh%20Component%20Array :type Array%20of%20Skeletal%20Mesh%20Component%20Object%20References))
  ;[1 1 1] Initialize the bp
  (Variables|Default|SetExternalMidArray MID%20Array)
  (Variables|Default|SetExternalSkeletalMeshComponent External%20MEsh%20Component%20Array)
  (Utilities|Array|Clear (Variables|Default|GetMIDArray)))

(event EventBeginPlay
  (bind _increment_end (Variables|Default|GetIncrement_End))
  ;[1 1 1] Initialize the bp
  (Variables|Default|GetExternalRootMeshComponent
    (:then
      (CallFunction|StoreOriginalMaterialForTeleportIn)
      (CallFunction|InitializeExternalSkeletalMeshArray)
      (CallFunction|FindBounds)
      (CallFunction|SpawnLight)
      (CallFunction|MakeAndSlaveInternalMeshes)
      (bind _corrected_mid_array (CallFunction|ExternalMeshMaskedSetup (Variables|Default|GetExternalSkeletalMeshComponent)))
      (Variables|Default|SetExternalMidArray _corrected_mid_array)
      (CallFunction|MakeInternalMidsAndAppendExternalAndIntermalMidsToMidArray)
      ;[1 1 1] Animation Speed
      (Actor|SetLifeSpan (Variables|Default|GetTimeline%20Play%20Length))
      (Components|Timeline|SetPlayRate (Variables|DuplicateResOutMesh|GetCharacterSpawnInTimeline) (/ 1 (Variables|Default|GetTimeline%20Play%20Length)))
      (if (Variables|Default|GetTeleportOut?)
        (bind (_direction _zhieght _lightintensity) (Timeline|CharacterSpawnInTimeline PlayFromStart)
          (:Update
            (CallFunction|ProcessSpawninTimeline _zhieght _lightintensity))
          (:Finished
            ;[1 1 1] Ensures Teleported In Characters have Opaque and Not Masked Materials during Gameplay
            (switch Utilities|FlowControl|Switch|SwitchonETimelineDirection _direction
              (:Backward
                (Utilities|IsValid (Variables|Default|GetExternalActor)
                  (:Is%20Valid
                    (for _array_element (Variables|Default|GetExternalSkeletalMeshComponent)
                      (for (_ _array_index) (Rendering|Material|GetMaterials _array_element)
                        (Rendering|Material|SetMaterial _array_element _array_index (Utilities|Array|Get(acopy) (Variables|Default|GetOriginal_MIDs) (+ _array_index _increment_end))))
                      (Variables|Default|SetIncrement_End (+ _increment_end (Utilities|Array|Length (Rendering|Material|GetMaterials _array_element)))))))))))
        (else
          (Timeline|CharacterSpawnInTimeline ReverseFromEnd))))
    (:else)))

(fn ExternalMeshMaskedSetup
  (params (External%20Mesh%20Component%20Array :type Array%20of%20Skeletal%20Mesh%20Component%20Object%20References))
  (bind _externalmidarray (Variables|Default|GetExternalMidArray))
  (bind _increment_maskedsetup (LocalVariables|GetIncrement_MaskedSetup))
  (LocalVariables|SetIncrement_MaskedSetup 0)
  (Utilities|Array|Clear (LocalVariables|GetMID_Array_MaskedSetup))
  (for _array_element External%20Mesh%20Component%20Array
    (for (_array_element_2 _array_index) (Rendering|Material|GetMaterials _array_element)
      (bind _asmaterial_instance (Utilities|Casting|CastToMaterialInstance _array_element_2))
      (if (or (Utilities|Array|ContainsItem (Variables|Default|GetListOfMaterialsExemptFromReparenting) _array_element_2) (Utilities|Array|ContainsItem (Variables|Default|GetListOfMaterialsExemptFromReparenting) (Variables|/Script/Engine.MaterialInstance|GetParent _asmaterial_instance)))
        (Development|PrintString (Utilities|GetDisplayName _array_element_2))
        (bind _returnvalue (Rendering|Material|CreateDynamicMaterialInstance _array_element _array_index _array_element_2 "None" :!class /Script/Engine.PrimitiveComponent))
        (Rendering|Material|CopyMaterialInstanceParameters _returnvalue (Utilities|Array|Get(acopy) (Variables|Default|GetExternalMidArray) (+ _array_index _increment_maskedsetup)) true)
        (Utilities|Array|Add (LocalVariables|GetMID_Array_MaskedSetup) _returnvalue)
        (else
          (bind _returnvalue_2 (Rendering|Material|CreateDynamicMaterialInstance _array_element _array_index (select (Variables|Default|GetGun%20True) "/Game/Packages/Fortnite_UniqueMaterials_1/Weapons/WeaponMat_Base_Masked.WeaponMat_Base_Masked" "/Game/Packages/Fortress_SharedMaterials/Base_Material/MI_CharacterShader_Masked.MI_CharacterShader_Masked") "None" :!class /Script/Engine.PrimitiveComponent))
          (Rendering|Material|CopyMaterialInstanceParameters _returnvalue_2 (Utilities|Array|Get(acopy) _externalmidarray (+ _array_index _increment_maskedsetup)) true)
          (Fortnite|CopyHideElements (Utilities|Array|Get(acopy) _externalmidarray (+ _array_index _increment_maskedsetup)) _returnvalue_2)
          (Fortnite|CopyJerseyElements (Utilities|Array|Get(acopy) _externalmidarray (+ _array_index _increment_maskedsetup)) _returnvalue_2)
          (Utilities|Array|Add (LocalVariables|GetMID_Array_MaskedSetup) _returnvalue_2))))
    (LocalVariables|SetIncrement_MaskedSetup (+ _increment_maskedsetup (Utilities|Array|Length (Rendering|Material|GetMaterials _array_element)))))
  (return (LocalVariables|GetMID_Array_MaskedSetup)))

(fn FindBounds
  (bind (_origin _boxextent) (Collision|GetActorBounds (Variables|Default|GetExternalActor) false false))
  ;[0 0 0] Bounds radius
  (bind _output_get (Variables|Default|SetBox%20Local%20Transform (Math|Transform|MakeTransform _origin "0, 0, 0" (* _boxextent "2.000000,2.000000,2.000000"))))
  (Variables|Default|SetBoundsRadius (Math|Vector|VectorLength (.scale (select (Variables|Default|GetGun%20True) (Math|Transform|MakeTransform (.location _output_get) (Math|Rotator|MakeRotationfromAxes (.Z (Math|Rotator|BreakRotIntoAxes (.rotation _output_get))) (Math|Rotator|BreakRotIntoAxes (.rotation _output_get)) (.Y (Math|Rotator|BreakRotIntoAxes (.rotation _output_get)))) (Math|Vector|MakeVector (.z (.scale _output_get)) (.x (.scale _output_get)) (.y (.scale _output_get)))) (Math|Transform|MakeTransform (.location _output_get) (.rotation _output_get) (Math|Vector|MakeVector 70 70 (* (.z (.scale _output_get)) 1.05)))))))
  (return))

(fn initializeExternalSkeletalMeshArray
  (bind _externalactor (Variables|Default|GetExternalActor))
  (bind (_asplayer_pawn_athena_generic_parent _bsuccess) (Utilities|Casting|CastToPlayerPawn_Athena_Generic_Parent _externalactor))
  (bind (_asfort_aipawn _bsuccess_2) (Utilities|Casting|CastToFortAIPawn _externalactor))
  (bind (_asplayer_pawn_generic_parent _bsuccess_3) (Utilities|Casting|CastToPlayerPawn_Generic_Parent _externalactor))
  (if _bsuccess_3
    (Variables|Default|SetExternal%20Base%20Mesh%20Component (Variables|/Script/Engine.Character|GetMesh _asplayer_pawn_generic_parent))
    (Variables|Default|SetExternalSkeletalMeshComponent (Variables|/Game/Abilities/Player/Pawns/PlayerPawn_Generic_Parent.PlayerPawn_Generic_Parent_C|GetSkeletalMeshes _asplayer_pawn_generic_parent))
    (elif _bsuccess
      (Variables|Default|SetExternal%20Base%20Mesh%20Component (Variables|/Script/Engine.Character|GetMesh _asplayer_pawn_athena_generic_parent))
      (Variables|Default|SetExternalSkeletalMeshComponent (Variables|/Game/Athena/PlayerPawn_Athena_Generic_Parent.PlayerPawn_Athena_Generic_Parent_C|GetSkeletalMeshes _asplayer_pawn_athena_generic_parent))
      (elif _bsuccess_2
        (Variables|Default|SetExternalSkeletalMeshComponent (Utilities|Array|MakeArray (Variables|/Script/Engine.Character|GetMesh _asfort_aipawn)))
        ;[1 0 0] this needs to be referenced more often instead of actual components
        (Variables|Default|SetExternal%20Base%20Mesh%20Component (Variables|/Script/Engine.Character|GetMesh _asfort_aipawn))
        (else
          (Actor|DestroyActor)))))
  (return))

(fn Makeandslaveinternalmeshes
  (bind _internalskeletalmeshcomponents (Variables|Default|GetInternalSkeletalMeshComponents))
  (LocalVariables|SetTransform (Math|Transform|MakeTransform "0, 0, 0"))
  (Components|DestroyComponent _internalskeletalmeshcomponents)
  (Utilities|Array|Clear _internalskeletalmeshcomponents)
  (for (_array_element _array_index) (Variables|Default|GetExternalSkeletalMeshComponent)
    (bind _returnvalue (AddComponent|NODE_AddSkeletalMeshComponent-0 :RelativeTransform (LocalVariables|GetTransform)))
    (Utilities|Array|Add (Variables|Default|GetInternalSkeletalMeshComponents) _returnvalue)
    (Components|SkinnedMesh|SetSkinnedAssetandUpdate _returnvalue (Components|SkeletalMesh|GetSkeletalMeshAsset _array_element) true)
    (Transformation|SetWorldTransform _returnvalue (Transformation|GetWorldTransform _array_element) false false)
    (Components|SkinnedMesh|SetLeaderPoseComponent (Utilities|Array|Get(acopy) (Variables|Default|GetInternalSkeletalMeshComponents) _array_index) (select (== _array_index 0) (Variables|Default|GetExternalRootMeshComponent) (Utilities|Array|Get(acopy) (Variables|Default|GetExternalSkeletalMeshComponent) _array_index)) false false)))

(fn Makeinternalmidsandappendexternalandintermalmidstomidarray
  (LocalVariables|SetGunOrCharacterMaterialInstance (select (Variables|Default|GetGun%20True) "/Game/Effects/Fort_Effects/Effects/Characters/Generic/ResIn/RezIn_Gun.RezIn_Gun" "/Game/Effects/Fort_Effects/Effects/Characters/Generic/ResIn/RezIn_Character.RezIn_Character"))
  (for (_array_element _array_index) (Variables|Default|GetExternalSkeletalMeshComponent)
    (LocalVariables|SetCurrentExternalSkeletalMeshComponent (Utilities|Array|Get(acopy) (Variables|Default|GetExternalSkeletalMeshComponent) _array_index))
    (bind _output_get (Variables|Default|SetNumber%20of%20Base%20skeletal%20mesh%20materials (- (Rendering|Material|GetNumMaterials _array_element) 1)))
    (for _index (range (+ _output_get 1))
      (Variables|Default|SetCurrentMeshMidIndex _index)
      (bind _returnvalue (Rendering|Material|CreateDynamicMaterialInstance (Utilities|Array|Get(acopy) (Variables|Default|GetInternalSkeletalMeshComponents) _array_index) (Variables|Default|GetCurrentMeshMidIndex) (LocalVariables|GetGunOrCharacterMaterialInstance) "None" :!class /Script/Engine.PrimitiveComponent))
      (Rendering|Material|CopyMaterialInstanceParameters _returnvalue (Rendering|Material|GetMaterial (LocalVariables|GetCurrentExternalSkeletalMeshComponent) (Variables|Default|GetCurrentMeshMidIndex)) true)
      (Utilities|Array|Add (Variables|Default|GetMIDArray) _returnvalue)
      (Rendering|Material|SetScalarParameterValue _returnvalue "BoundsRadius" (Variables|Default|GetBoundsRadius) :!class /Script/Engine.MaterialInstanceDynamic))
    ;[0 0 0] Add External MIDs to the list
    (Utilities|Array|AppendArray (Variables|Default|GetMIDArray) (Variables|Default|GetExternalMidArray))))

(fn ProcessSpawnInTimeline
  (params (ZHeight :type Float%20%28double-precision%29) (LightIntensity :type Float%20%28double-precision%29))
  (bind _returnvalue 1)
  (bind _returnvalue_2 (Utilities|Name|MakeLiteralName "Root_Effect"))
  (bind _returnvalue_3 (Utilities|Name|MakeLiteralName "FX_Head"))
  (bind _head_space (Variables|Default|GetHead%20Space))
  (bind _externalrootmeshcomponent (Variables|Default|GetExternalRootMeshComponent))
  (bind _returnvalue_4 (Transformation|GetSocketLocation _externalrootmeshcomponent _returnvalue_2))
  (bind _returnvalue_5 (Transformation|GetSocketLocation _externalrootmeshcomponent _returnvalue_3))
  (bind _returnvalue_6 (- _returnvalue_5 _returnvalue_4))
  (bind _returnvalue_7 (Math|Vector|Normalize _returnvalue_6))
  (bind _returnvalue_8 (* _returnvalue_7 _head_space))
  (bind _leg_space (Variables|Default|GetLeg%20Space))
  (bind _returnvalue_9 (* _returnvalue_7 _leg_space))
  (bind _returnvalue_10 (+ _returnvalue_9 _returnvalue_4))
  (bind _returnvalue_11 (+ _returnvalue_5 _returnvalue_8))
  (bind _current_mid (Variables|Default|GetCurrent%20MID))
  (bind _spawn_light (Variables|Default|GetSpawn%20Light))
  (bind _firstpasscomplete (Variables|Default|GetFirstPassComplete))
  (Variables|Default|GetExternalRootMeshComponent
    (:then
      (Variables|Default|SetFirstPassComplete _returnvalue)
      (LocalVariables|SetDissolveGradientPercentage (Math|Float|Lerp ZHeight (- 1.0 ZHeight) _returnvalue))
      (LocalVariables|SetDissolveGradientDistance (Math|Vector|VectorLength (- _returnvalue_10 _returnvalue_11)))
      (Variables|Default|SetWorld%20location%20for%20the%20dissolve%20location (Math|Vector|Lerp(Vector) _returnvalue_11 _returnvalue_10 ZHeight))
      (LocalVariables|SetDissolveGradientDirectionalColor (Math|Conversions|ToLinearColor(Vector) (* _returnvalue_7 (Math|Float|Lerp 1 -1 _firstpasscomplete))))
      (LocalVariables|SetDissolveGradientBaseLocation (Math|Conversions|ToLinearColor(Vector) (Math|Vector|Lerp(Vector) _returnvalue_10 _returnvalue_11 _firstpasscomplete)))
      (Utilities|IsValid _spawn_light
        (:Is%20Valid
          (Transformation|SetWorldLocation _spawn_light (Variables|Default|GetWorld%20location%20for%20the%20dissolve%20location) false false)
          (Rendering|Components|Light|SetIntensity _spawn_light (* LightIntensity (Variables|Default|Getlight%20intensity))))
        (:Is%20Not%20Valid
          _))
      (for _array_element (Variables|Default|GetMIDArray)
        (bind _output_get (Variables|Default|SetCurrent%20MID _array_element))
        (Rendering|Material|SetScalarParameterValue _output_get "DissolveGradientPercentage" (LocalVariables|GetDissolveGradientPercentage) :!class /Script/Engine.MaterialInstanceDynamic)
        ;[1 1 1] Make it accomodate Weapons as well
        (Rendering|Material|SetVectorParameterValue (Variables|Default|GetCurrent%20MID) "DissolveGradientDirectionalVector" (LocalVariables|GetDissolveGradientDirectionalColor) :!class /Script/Engine.MaterialInstanceDynamic)
        ;[1 1 1] get the base of the object
        (Rendering|Material|SetVectorParameterValue (Variables|Default|GetCurrent%20MID) "DissolveGradientBaseLocation" (LocalVariables|GetDissolveGradientBaseLocation) :!class /Script/Engine.MaterialInstanceDynamic)
        (Rendering|Material|SetScalarParameterValue _current_mid "DissolveGradientDistance" (LocalVariables|GetDissolveGradientDistance) :!class /Script/Engine.MaterialInstanceDynamic)
        (Rendering|Material|SetScalarParameterValue _current_mid "FirstSpawnPassComplete" (Variables|Default|GetFirstPassComplete) :!class /Script/Engine.MaterialInstanceDynamic)))
    (:else
      (Components|Timeline|Stop (Variables|DuplicateResOutMesh|GetCharacterSpawnInTimeline)))))

(fn RestoreInProgressRes
  (CallFunction|InitializeExternalSkeletalMeshArray)
  (CallFunction|MakeAndSlaveInternalMeshes)
  (bind _corrected_mid_array (CallFunction|ExternalMeshMaskedSetup (Variables|Default|GetExternalSkeletalMeshComponent)))
  (Variables|Default|SetExternalMidArray _corrected_mid_array)
  (CallFunction|MakeInternalMidsAndAppendExternalAndIntermalMidsToMidArray))

(fn SpawnLight
  ;[1 1 1] No Light Spawned on Mobile
  (if (Fortnite|IsMobilePlatform)
    _
    (elif (Variables|Default|GetGun%20True)
      ;[0 0 0] point light
      (bind _returnvalue (AddComponent|NODE_AddPointLightComponent-0 :RelativeTransform (Math|Transform|MakeTransform "0.000000,0.000000,100.000000")))
      (bind _output_get (Variables|Default|SetSpawn%20Light _returnvalue))
      (Rendering|Components|Light|SetIntensity _output_get (Variables|Default|Getlight%20intensity)))))

(fn StoreOriginalMaterialforTeleportIn
  (Utilities|Array|Clear (Variables|Default|GetOriginal_MIDs))
  (Variables|Default|SetOriginal_MIDs (Variables|Default|GetExternalMidArray)))

(fn ConstructionScript)
