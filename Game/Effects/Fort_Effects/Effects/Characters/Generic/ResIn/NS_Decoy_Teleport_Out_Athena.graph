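# Niagara system for a short cyan-blue teleport-out burst on a decoy character.
# Single-particle core flare, corona and glow sprites flash briefly while mesh fragments and GPU sparkle sprites spawn from the User.SkeletalMesh surface in staggered bursts and continuous streams.
# Fragments and sparkles use negatively scaled inherited velocity, noise, downward acceleration and flickering or curved size fades.
# Shimmer sprites sampled from the mesh rise and are pulled toward User.PointPosition.
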NiagaraSystem /Game/Effects/Fort_Effects/Effects/Characters/Generic/ResIn/NS_Decoy_Teleport_Out_Athena
dependencies {
 usedDataInterfaces [&NiagaraDataInterfaceCurve, &NiagaraDataInterfaceVectorCurve, &NiagaraDataInterfaceColorCurve]
 usedDynamicInputs []
 usedModules [&SystemState]
}
emitters {
 CoreFlare {
  emitterUpdateScript [
   &EmitterState(Life Cycle Mode=::NewEnumerator1, Loop Behavior=::NewEnumerator1, Loop Duration=0.5, UseLoopDelay=false)
   &SpawnBurst_Instantaneous(Spawn Count=1, Use Spawn Probability=false)
  ]
  particleSpawnScript [
   /Niagara/Modules/Spawn/Initialization/InitializeParticle(
    Color=<rgba 0.2123 3.8438 5 0.1>
    Lifetime=0.25
    Module.Write Color=true
    Module.Write Lifetime=true
    Module.Write Mass=true
    Module.Write Material Random=true
    Module.Write Position=false
    Module.Write Scale=false
    Module.Write SpriteRotation=false
    Module.Write SpriteSize=true
    Position=&SimulationPosition()
    Sprite Size=&Vector2DFromFloat(Value=1500)
   )
   &CameraOffset(Camera Offset Amount=50)
   &DynamicMaterialParameters(
    Index 0 Param 1=0.1
    Index 0 Param 2=1
    Index 0 Param 3=1
    Index 0 Param 4=1
    Module.Param0WriteEnabled=true
    Module.Param1WriteEnabled=true
    Module.Param2WriteEnabled=true
    Module.Param3WriteEnabled=true
   )
  ]
  particleUpdateScript [
   &ParticleState()
   &SolveForcesAndVelocity(Clamp Velocity=false, Limit Acceleration=false)
   &ScaleColor(Scale Alpha=0.7, ScaleA=true, ScaleRGB=true)
   &ScaleSpriteSize(Scale Factor=&Vector2DFromFloat(Value=&FloatFromCurve(FloatCurve=DataInterface() {propertyValues {Curve {keys [
          {arriveTangent=3310.2024, interpMode=RCIM_Cubic, time=0, value=0}
          {arriveTangent=-15.1945, interpMode=RCIM_Cubic, leaveTangent=-3.0716, tangentMode=RCTM_Break, time=0.15, value=1}
          {interpMode=RCIM_Cubic, leaveTangent=-1.6445, time=1, value=0}
         ]}}})))
  ]
  rendererClasses [&NiagaraSpriteRendererProperties]
  renderers [{rendererClass=&NiagaraSpriteRendererProperties}]
 }
 CoreFlare_MedHigh {
  emitterUpdateScript [
   &EmitterState(Life Cycle Mode=::NewEnumerator1, Loop Behavior=::NewEnumerator1, Loop Duration=0.5, UseLoopDelay=false)
   &SpawnBurst_Instantaneous(Spawn Count=1, Use Spawn Probability=false)
  ]
  particleSpawnScript [
   &InitializeParticle(
    Color=<rgba 0.2123 3.8438 5 0.25>
    Lifetime=0.25
    Module.Write Color=true
    Module.Write Lifetime=true
    Module.Write Mass=true
    Module.Write Material Random=true
    Module.Write Position=false
    Module.Write Scale=false
    Module.Write SpriteRotation=false
    Module.Write SpriteSize=true
    Position=&SimulationPosition()
    Sprite Size=&Vector2DFromFloat(Value=1500)
   )
   &CameraOffset(Camera Offset Amount=100)
   &DynamicMaterialParameters(
    Index 0 Param 1=0.1
    Index 0 Param 2=1
    Index 0 Param 3=1
    Index 0 Param 4=1
    Module.Param0WriteEnabled=true
    Module.Param1WriteEnabled=true
    Module.Param2WriteEnabled=true
    Module.Param3WriteEnabled=true
   )
  ]
  particleUpdateScript [
   &ParticleState()
   &SolveForcesAndVelocity(Clamp Velocity=false, Limit Acceleration=false)
   &ScaleColor(Scale Alpha=0.7, ScaleA=true, ScaleRGB=true)
   &ScaleSpriteSize(Scale Factor=&Vector2DFromFloat(Value=&FloatFromCurve(FloatCurve=DataInterface() {propertyValues {Curve {keys [
          {arriveTangent=3310.2024, interpMode=RCIM_Cubic, time=0, value=0}
          {arriveTangent=0.7273, interpMode=RCIM_Cubic, leaveTangent=0.7273, time=0.1, value=0.3}
          {arriveTangent=-0.3333, interpMode=RCIM_Cubic, leaveTangent=-0.3333, time=0.55, value=0.4}
          {interpMode=RCIM_Cubic, leaveTangent=-1.6445, time=1, value=0}
         ]}}})))
  ]
  rendererClasses [&NiagaraSpriteRendererProperties]
  renderers [{rendererClass=&NiagaraSpriteRendererProperties}]
 }
 Corona {
  emitterUpdateScript [
   &EmitterState()
   &SpawnBurst_Instantaneous(Spawn Count=1, Use Spawn Probability=false)
  ]
  particleSpawnScript [
   &InitializeParticle(
    Color=<rgba 0.6721 2.1415 5 0.25>
    Lifetime=0.15
    Module.Write Color=true
    Module.Write Lifetime=true
    Module.Write Mass=true
    Module.Write Material Random=true
    Module.Write Position=false
    Module.Write Scale=false
    Module.Write SpriteRotation=false
    Module.Write SpriteSize=true
    Position=&SimulationPosition()
    Sprite Size=&Vector2DFromFloat(Value=300)
   )
   &CameraOffset(Camera Offset Amount=1)
  ]
  particleUpdateScript [
   &ParticleState()
   &SolveForcesAndVelocity(Clamp Velocity=false, Limit Acceleration=false)
   &DynamicMaterialParameters(
    Index 0 Param 1=&FloatFromCurve(FloatCurve=DataInterface() {propertyValues {Curve {keys [{arriveTangent=-2.7401, interpMode=RCIM_Cubic, leaveTangent=-2.7401, tangentMode=RCTM_User, time=0, value=2}, <at 1 1 RCIM_Cubic>]}}})
    Index 0 Param 2=0.4
    Index 0 Param 3=0.6
    Index 0 Param 4=1
    Module.Param0WriteEnabled=true
    Module.Param1WriteEnabled=true
    Module.Param2WriteEnabled=true
    Module.Param3WriteEnabled=true
   )
   &ScaleColor(
    Scale Alpha=&FloatFromCurve(FloatCurve=DataInterface() {propertyValues {Curve {keys [<at 0 0 RCIM_Cubic>,<at 0.25 1 RCIM_Cubic>,<at 1 0 RCIM_Cubic>]}}})
    Scale RGB=<2 2 2>
    ScaleA=true
    ScaleRGB=true
   )
   &ScaleSpriteSize(Scale Factor=&Vector2DFromFloat(Value=&FloatFromCurve(FloatCurve=DataInterface() {propertyValues {Curve {keys [
          {arriveTangent=3310.2024, interpMode=RCIM_Cubic, time=0, value=0.1}
          {arriveTangent=1.4167, interpMode=RCIM_Cubic, leaveTangent=1.4167, time=0.4, value=0.8}
          {arriveTangent=0.3333, interpMode=RCIM_Cubic, leaveTangent=0.3333, time=0.6, value=0.95}
          {interpMode=RCIM_Cubic, leaveTangent=-1.6445, time=1, value=1}
         ]}}})))
  ]
  rendererClasses [&NiagaraSpriteRendererProperties]
  renderers [{rendererClass=&NiagaraSpriteRendererProperties}]
 }
 Fragments {
  emitterUpdateScript [
   &EmitterState(Life Cycle Mode=::NewEnumerator1, Loop Behavior=::NewEnumerator1, Loop Duration=0.75, UseLoopDelay=false)
   &SpawnRate(SpawnRate=50, Use Spawn Probability=false)
  ]
  particleSpawnScript [
   &InitializeParticle(
    Color=&MakeLinearColorFromVectorAndFloat("Vector (RGB)"=<25 25 25>, "Float (Alpha)"=&UniformRangedFloat(Minimum=0.7, Maximum=1))
    Lifetime=&UniformRangedFloat(Minimum=0.25, Maximum=1)
    Mesh Scale=&VectorFromFloat(Value=&UniformRangedFloat(Minimum=0.1, Maximum=0.3))
    Module.Write Color=true
    Module.Write Lifetime=true
    Module.Write Mass=true
    Module.Write Material Random=true
    Module.Write Position=false
    Module.Write Scale=true
    Module.Write SpriteRotation=false
    Module.Write SpriteSize=false
    Position=&SimulationPosition()
    Sprite Size=<5 5>
   )
   &SkeletalMeshLocation(
    Inherit Velocity=true
    "Inherit Velocity (Scale)"=-0.2
    MaskSpawnGroup=false
    Module.Override Randomness=false
    Skeletal Mesh=$User.SkeletalMesh
    VelocitySpeedLimit=false
   )
   &SphereLocation(
    MaskSpawnGroup=false
    Override Seed=false
    Sphere Radius=10
    Surface Only=false
    UseNonUniformScale=false
    UseOffset=false
    UseSphereOrientation=false
   )
   &AddVelocity(Velocity=&Multiply_Vector(
     A=$Output.SphereLocation.SphereNormal
     B=&VectorFromFloat(Value=&UniformRangedFloat(Minimum=1, Maximum=3))
    ))
   /Niagara/Modules/Spawn/Orientation/V2/InitialMeshOrientation(
    Rotation=&UniformRangedVector(Minimum=<0 0 0>, Maximum=<1 1 1>)
    Use Orientation Vector=true
    Use Rotation Vector=true
   )
   &AddVelocityInCone(
    Cone Angle=&NormalizedAngleToDegrees(NormalizedAngle=&Lerp_Float(A=0, B=1, Alpha=$Particles.NormalizedAge))
    Cone Axis=<0 0 -1>
    Module.Use Velocity Falloff On Cone Axis=false
    Velocity Strength=200
   )
  ]
  particleUpdateScript [
   &ParticleState()
   &AccelerationForce(Acceleration=<0 0 -300>)
   &VectorNoiseForce(
    Force Amount=&UniformRangedFloat(Minimum=1250, Maximum=2500)
    Module.Use Random Noise Time Interval=true
    Random Noise Time Interval=0.025
   )
   &SolveForcesAndVelocity(Clamp Velocity=false, Limit Acceleration=false)
   /Niagara/Modules/Update/Orientation/V2/UpdateMeshOrientation(
    Rotation Rate:NiagaraFloat=&FloatFromCurve(FloatCurve=DataInterface() {propertyValues {Curve {keys [{arriveTangent=-2.8458, interpMode=RCIM_Cubic, leaveTangent=-2.8458, tangentMode=RCTM_User, time=0, value=1}, <at 1 0.2 RCIM_Cubic>]}}})
    Rotation Vector:Vector3f=<2 0.4 0.1>
   )
   &ScaleMeshSize(Scale Factor=&VectorFromFloat(Value=&FloatFromCurve(FloatCurve=DataInterface() {propertyValues {Curve {keys [<at 0 0 RCIM_Cubic>,<at 0.25 2 RCIM_Cubic>,<at 1 0 RCIM_Cubic>]}}})))
   &ScaleColor(
    Scale Alpha=&FloatFromCurve(FloatCurve=DataInterface() {propertyValues {Curve {keys [<at 0.7 1 RCIM_Cubic>,<at 1 0>]}}})
    Scale RGB=&MakeVectorFromLinearColorRGB(Color=&ColorFromCurve(ColorCurve=DataInterface() {propertyValues {
        AlphaCurve {keys []}
        BlueCurve {keys [<at 0 8 RCIM_Cubic>,<at 0.8 0.25>,<at 1 0 RCIM_Cubic>]}
        GreenCurve {keys [<at 0 1 RCIM_Cubic>,<at 0.8 0.25>,<at 1 0 RCIM_Cubic>]}
        RedCurve {keys [<at 0 1 RCIM_Cubic>,<at 0.8 0.25>,<at 1 0 RCIM_Cubic>]}
       }}))
    ScaleA=true
    ScaleRGB=true
   )
  ]
  rendererClasses [&NiagaraMeshRendererProperties]
  renderers [{rendererClass=&NiagaraMeshRendererProperties, rendererIndex=0}]
 }
 Fragments_Burst {
  emitterUpdateScript [
   &EmitterState(Life Cycle Mode=::NewEnumerator1, Loop Behavior=::NewEnumerator1, Loop Duration=0.75, UseLoopDelay=false)
   &SpawnBurst_Instantaneous(Spawn Count=5, Use Spawn Probability=false)
   &SpawnBurst_Instantaneous(Spawn Count=25, Spawn Time=0.02, Use Spawn Probability=false) {module=SpawnBurst_Instantaneous001}
   &SpawnBurst_Instantaneous(Spawn Count=50, Spawn Time=0.04, Use Spawn Probability=false) {module=SpawnBurst_Instantaneous002}
  ]
  particleSpawnScript [
   &InitializeParticle(
    Color=&MakeLinearColorFromVectorAndFloat("Vector (RGB)"=<25 25 25>, "Float (Alpha)"=&UniformRangedFloat(Minimum=0.7, Maximum=1))
    Lifetime=&UniformRangedFloat(Minimum=0.25, Maximum=1)
    Mesh Scale=&VectorFromFloat(Value=&UniformRangedFloat(Minimum=0.1, Maximum=0.3))
    Module.Write Color=true
    Module.Write Lifetime=true
    Module.Write Mass=true
    Module.Write Material Random=true
    Module.Write Position=false
    Module.Write Scale=true
    Module.Write SpriteRotation=false
    Module.Write SpriteSize=false
    Position=&SimulationPosition()
    Sprite Size=<5 5>
   )
   &SkeletalMeshLocation(
    Inherit Velocity=true
    "Inherit Velocity (Scale)"=-1
    MaskSpawnGroup=false
    Module.Override Randomness=false
    Skeletal Mesh=$User.SkeletalMesh
    VelocitySpeedLimit=false
   )
   &SphereLocation(
    MaskSpawnGroup=false
    Override Seed=false
    Sphere Radius=&UniformRangedFloat(Minimum=0, Maximum=20)
    Surface Only=false
    UseNonUniformScale=false
    UseOffset=false
    UseSphereOrientation=false
   )
   &AddVelocity(Velocity=&Multiply_Vector(
     A=$Output.SphereLocation.SphereNormal
     B=&VectorFromFloat(Value=&FloatFromCurve(
       FloatCurve=DataInterface() {propertyValues {Curve {keys [<at 0 36 RCIM_Cubic>, {arriveTangent=-36, interpMode=RCIM_Cubic, leaveTangent=-36, time=0.37, value=12}, <at 1 0 RCIM_Cubic>]}}}
       Scale Curve=5
      ))
    ))
   &AddVelocityInCone(
    Cone Angle=&NormalizedAngleToDegrees(NormalizedAngle=&Lerp_Float(A=0, B=1, Alpha=$Particles.NormalizedAge))
    Cone Axis=<0 0 -1>
    Module.Use Velocity Falloff On Cone Axis=false
    Velocity Strength=100
   )
   &InitialMeshOrientation(
    Rotation=&UniformRangedVector(Minimum=<0 0 0>, Maximum=<1 1 1>)
    Use Orientation Vector=true
    Use Rotation Vector=true
   )
  ]
  particleUpdateScript [
   &ParticleState()
   &AccelerationForce(Acceleration=<0 0 -180>)
   &VectorNoiseForce(
    Force Amount=&UniformRangedFloat(Minimum=625, Maximum=2000)
    Module.Use Random Noise Time Interval=true
    Random Noise Time Interval=0.15
   ) {enabled=false}
   &Drag(Use Linear Drag=true, Use Rotational Drag=false)
   &SolveForcesAndVelocity(Clamp Velocity=false, Limit Acceleration=false)
   &UpdateMeshOrientation(
    Rotation Rate:NiagaraFloat=&FloatFromCurve(FloatCurve=DataInterface() {propertyValues {Curve {keys [{arriveTangent=-2.8458, interpMode=RCIM_Cubic, leaveTangent=-2.8458, tangentMode=RCTM_User, time=0, value=1}, <at 1 0.2 RCIM_Cubic>]}}})
    Rotation Vector:Vector3f=<2 0.4 0.1>
   )
   &ScaleMeshSize(Scale Factor=&VectorFromFloat(Value=&FloatFromCurve(FloatCurve=DataInterface() {propertyValues {Curve {keys [<at 0 0 RCIM_Cubic>,<at 0.25 2 RCIM_Cubic>,<at 1 0 RCIM_Cubic>]}}})))
   &ScaleColor(
    Scale Alpha=&FloatFromCurve(FloatCurve=DataInterface() {propertyValues {Curve {keys [<at 0.7 1 RCIM_Cubic>,<at 1 0>]}}})
    Scale RGB=&MakeVectorFromLinearColorRGB(Color=&ColorFromCurve(ColorCurve=DataInterface() {propertyValues {
        AlphaCurve {keys []}
        BlueCurve {keys [<at 0 8 RCIM_Cubic>,<at 0.8 0.25>,<at 1 0 RCIM_Cubic>]}
        GreenCurve {keys [<at 0 1 RCIM_Cubic>,<at 0.8 0.25>,<at 1 0 RCIM_Cubic>]}
        RedCurve {keys [<at 0 1 RCIM_Cubic>,<at 0.8 0.25>,<at 1 0 RCIM_Cubic>]}
       }}))
    ScaleA=true
    ScaleRGB=true
   )
  ]
  rendererClasses [&NiagaraMeshRendererProperties]
  renderers [{rendererClass=&NiagaraMeshRendererProperties, rendererIndex=0}]
 }
 Fragments_TightoBody {
  emitterUpdateScript [
   &EmitterState(Life Cycle Mode=::NewEnumerator1, Loop Behavior=::NewEnumerator1, Loop Duration=0.75, UseLoopDelay=false)
   &SpawnRate(
    SpawnRate=&FloatFromCurve(FloatCurve=DataInterface() {propertyValues {Curve {keys [<at 0 50>,<at 0.7 100>]}}}, CurveIndex=$Emitter.NormalizedLoopAge)
    Use Spawn Probability=false
   )
  ]
  particleSpawnScript [
   &InitializeParticle(
    Color=&MakeLinearColorFromVectorAndFloat("Vector (RGB)"=<25 25 25>, "Float (Alpha)"=&UniformRangedFloat(Minimum=0.7, Maximum=1))
    Lifetime=&UniformRangedFloat(Minimum=0.25, Maximum=1)
    Mesh Scale=&VectorFromFloat(Value=&UniformRangedFloat(Minimum=0.1, Maximum=0.3))
    Module.Write Color=true
    Module.Write Lifetime=true
    Module.Write Mass=true
    Module.Write Material Random=true
    Module.Write Position=false
    Module.Write Scale=true
    Module.Write SpriteRotation=false
    Module.Write SpriteSize=false
    Position=&SimulationPosition()
    Sprite Size=<5 5>
   )
   &SkeletalMeshLocation(
    Inherit Velocity=true
    "Inherit Velocity (Scale)"=-1
    MaskSpawnGroup=false
    Module.Override Randomness=false
    Skeletal Mesh=$User.SkeletalMesh
    VelocitySpeedLimit=false
   )
   &SphereLocation(
    MaskSpawnGroup=false
    Override Seed=false
    Sphere Radius=&UniformRangedFloat(Minimum=0, Maximum=20)
    Surface Only=false
    UseNonUniformScale=false
    UseOffset=false
    UseSphereOrientation=false
   )
   &AddVelocity(Velocity=&Multiply_Vector(
     A=$Output.SphereLocation.SphereNormal
     B=&VectorFromFloat(Value=&UniformRangedFloat(Minimum=0, Maximum=1))
    ))
   &InitialMeshOrientation(
    Rotation=&UniformRangedVector(Minimum=<0 0 0>, Maximum=<1 1 1>)
    Use Orientation Vector=true
    Use Rotation Vector=true
   )
  ]
  particleUpdateScript [
   &ParticleState()
   &AddVelocityInCone(
    Cone Axis=<0 0 -1>
    Module.Use Velocity Falloff On Cone Axis=false
    Velocity Strength=&FloatFromCurve(
     FloatCurve=DataInterface() {propertyValues {Curve {keys [{arriveTangent=-3.9118, interpMode=RCIM_Cubic, leaveTangent=-3.9118, tangentMode=RCTM_User, time=0, value=1}, <at 1 -1 RCIM_Cubic>]}}}
     Scale Curve=50
    )
   )
   &AccelerationForce(Acceleration=<0 0 -180>)
   &VectorNoiseForce(
    Force Amount=&UniformRangedFloat(Minimum=625, Maximum=2000)
    Module.Use Random Noise Time Interval=true
    Random Noise Time Interval=0.15
   )
   &Drag(
    Drag=&Multiply_Float(
     A=&UniformRangedFloat(Minimum=1, Maximum=2)
     B=&Lerp_Float(A=1, B=2.5, Alpha=$Particles.NormalizedAge)
    )
    Use Linear Drag=true
    Use Rotational Drag=false
   )
   &SolveForcesAndVelocity(Clamp Velocity=false, Limit Acceleration=false)
   &UpdateMeshOrientation(
    Rotation Rate:NiagaraFloat=&FloatFromCurve(FloatCurve=DataInterface() {propertyValues {Curve {keys [{arriveTangent=-2.8458, interpMode=RCIM_Cubic, leaveTangent=-2.8458, tangentMode=RCTM_User, time=0, value=1}, <at 1 0.2 RCIM_Cubic>]}}})
    Rotation Vector:Vector3f=<2 0.4 0.1>
   )
   &ScaleMeshSize(Scale Factor=&VectorFromFloat(Value=&FloatFromCurve(FloatCurve=DataInterface() {propertyValues {Curve {keys [<at 0 0 RCIM_Cubic>,<at 0.25 2 RCIM_Cubic>,<at 1 0 RCIM_Cubic>]}}})))
   &ScaleColor(
    Scale Alpha=&FloatFromCurve(FloatCurve=DataInterface() {propertyValues {Curve {keys [<at 0.7 1 RCIM_Cubic>,<at 1 0>]}}})
    Scale RGB=&MakeVectorFromLinearColorRGB(Color=&ColorFromCurve(ColorCurve=DataInterface() {propertyValues {
        AlphaCurve {keys []}
        BlueCurve {keys [<at 0 8 RCIM_Cubic>,<at 0.8 0.25>,<at 1 0 RCIM_Cubic>]}
        GreenCurve {keys [<at 0 1 RCIM_Cubic>,<at 0.8 0.25>,<at 1 0 RCIM_Cubic>]}
        RedCurve {keys [<at 0 1 RCIM_Cubic>,<at 0.8 0.25>,<at 1 0 RCIM_Cubic>]}
       }}))
    ScaleA=true
    ScaleRGB=true
   )
  ]
  rendererClasses [&NiagaraMeshRendererProperties]
  renderers [{rendererClass=&NiagaraMeshRendererProperties, rendererIndex=0}]
 }
 Glow {
  emitterUpdateScript [
   &EmitterState()
   &SpawnBurst_Instantaneous(Spawn Count=1, Use Spawn Probability=false)
  ]
  particleSpawnScript [
   &InitializeParticle(
    Color=<rgba 0 0 0 0.75>
    Lifetime=0.25
    Module.Write Color=true
    Module.Write Lifetime=true
    Module.Write Mass=true
    Module.Write Material Random=true
    Module.Write Position=false
    Module.Write Scale=false
    Module.Write SpriteRotation=false
    Module.Write SpriteSize=true
    Position=&SimulationPosition()
    Sprite Size=&Vector2DFromFloat(Value=&UniformRangedFloat(Minimum=400, Maximum=500))
   )
   &CameraOffset(Camera Offset Amount=25)
   &DynamicMaterialParameters(
    Index 0 Param 1=1
    Index 0 Param 2=1
    Index 0 Param 3=1
    Index 0 Param 4=1
    Module.Param0WriteEnabled=true
    Module.Param1WriteEnabled=true
    Module.Param2WriteEnabled=true
    Module.Param3WriteEnabled=true
   )
  ]
  particleUpdateScript [
   &ParticleState()
   &SolveForcesAndVelocity(Clamp Velocity=false, Limit Acceleration=false)
   &ScaleColor(
    Scale Alpha=&FloatFromCurve(FloatCurve=DataInterface() {propertyValues {Curve {keys [
         <at 0 0 RCIM_Cubic>
         {arriveTangent=1.925, interpMode=RCIM_Cubic, leaveTangent=1.925, time=0.2, value=0.55}
         {arriveTangent=-0.4615, interpMode=RCIM_Cubic, leaveTangent=-0.4615, time=0.4, value=0.77}
         {arriveTangent=-1.2833, interpMode=RCIM_Cubic, leaveTangent=-1.2833, time=0.85, value=0.25}
         <at 1 0 RCIM_Cubic>
        ]}}})
    ScaleA=true
    ScaleRGB=true
   )
   &ScaleSpriteSize(Scale Factor=&Vector2DFromFloat(Value=&FloatFromCurve(FloatCurve=DataInterface() {propertyValues {Curve {keys [{arriveTangent=1.9842, interpMode=RCIM_Cubic, leaveTangent=1.9842, tangentMode=RCTM_User, time=0, value=0.35}, <at 1 1 RCIM_Cubic>]}}})))
  ]
  rendererClasses [&NiagaraSpriteRendererProperties]
  renderers [{rendererClass=&NiagaraSpriteRendererProperties}]
 }
 Shimmer {
  emitterUpdateScript [
   &EmitterState(Life Cycle Mode=::NewEnumerator1, Loop Behavior=::NewEnumerator1, Loop Duration=0.75, UseLoopDelay=false)
   &SpawnRate(SpawnRate=20, Use Spawn Probability=false)
  ]
  particleSpawnScript [
   &InitializeParticle(
    Color=&ColorFromCurve(
     ColorCurve=DataInterface() {propertyValues {
       AlphaCurve {keys [<at 0 1 RCIM_Cubic>,<at 1 0.5 RCIM_Cubic>]}
       BlueCurve {keys [<at 0 1 RCIM_Cubic>, {arriveTangent=-0.75, interpMode=RCIM_Cubic, leaveTangent=-0.75, time=0.3489, value=0.7122}, <at 1 0.25 RCIM_Cubic>]}
       GreenCurve {keys [<at 0 1 RCIM_Cubic>, {arriveTangent=-0.75, interpMode=RCIM_Cubic, leaveTangent=-0.75, time=0.3489, value=0.7122}, <at 1 0.25 RCIM_Cubic>]}
       RedCurve {keys [
         <at 0 1.371 RCIM_Cubic>
         {arriveTangent=-1.121, interpMode=RCIM_Cubic, leaveTangent=-1.121, time=0.3489, value=0.5965}
         <at 1 0.25 RCIM_Cubic>
        ]}
      }}
     CurveIndex=$Emitter.LoopedAge
    )
    Lifetime=&FloatFromCurve(FloatCurve=DataInterface() {propertyValues {Curve {keys [<at 0 0.4>,<at 1 0.5>]}}}, CurveIndex=$Emitter.NormalizedLoopAge)
    Module.Write Color=true
    Module.Write Lifetime=true
    Module.Write Mass=true
    Module.Write Material Random=true
    Module.Write Position=false
    Module.Write Scale=false
    Module.Write SpriteRotation=true
    Module.Write SpriteSize=true
    Position=&SimulationPosition()
    Sprite Rotation=&NormalizedAngleToDegrees(NormalizedAngle=&UniformRangedFloat(Minimum=0, Maximum=1))
    Sprite Size=&Vector2DFromFloat(Value=250)
   )
   &DynamicMaterialParameters(
    Index 0 Param 1=&UniformRangedFloat(Minimum=0, Maximum=1)
    Index 0 Param 2=&UniformRangedFloat(Minimum=0, Maximum=1)
    Index 0 Param 3=&UniformRangedFloat(Minimum=0, Maximum=1)
    Index 0 Param 4=1
    Module.Param0WriteEnabled=true
    Module.Param1WriteEnabled=true
    Module.Param2WriteEnabled=true
    Module.Param3WriteEnabled=true
   )
   &CameraOffset(Camera Offset Amount=&UniformRangedFloat(Minimum=-25, Maximum=25))
   &SkeletalMeshLocation(
    Bone Sampling Mode=::NewEnumerator0
    Inherit Velocity=false
    MaskSpawnGroup=false
    Module.Override Randomness=false
    Skeletal Mesh=$User.SkeletalMesh
   )
   &SphereLocation(
    MaskSpawnGroup=false
    Override Seed=false
    Sphere Radius=25
    Surface Only=false
    UseNonUniformScale=false
    UseOffset=false
    UseSphereOrientation=false
   )
   &AddVelocity(Velocity=&Multiply_Vector(
     A=$Output.SphereLocation.SphereNormal
     B=&VectorFromFloat(Value=&UniformRangedFloat(Minimum=-0.04, Maximum=0.02))
    ))
  ]
  particleUpdateScript [
   &ParticleState()
   &AccelerationForce(Acceleration=&UniformRangedVector(Minimum=<0 0 300>, Maximum=<0 0 500>))
   /Niagara/Modules/Update/Forces/PointAttractionForce(
    Attraction Radius=512
    AttractionStrength=100
    AttractorPosition:Vector3f=$User.PointPosition
    Module.Kill Within Radius=false
    Module.Use Falloff=true
   )
   &SolveForcesAndVelocity(Clamp Velocity=false, Limit Acceleration=false)
   &ScaleSpriteSize(Scale Factor=&Vector2DFromFloat(Value=&FloatFromCurve(FloatCurve=DataInterface() {propertyValues {Curve {keys [
          {arriveTangent=1.9218, interpMode=RCIM_Cubic, leaveTangent=1.9218, tangentMode=RCTM_User, time=0, value=0}
          {arriveTangent=1, interpMode=RCIM_Cubic, leaveTangent=1, time=0.45, value=0.85}
          <at 1 1 RCIM_Cubic>
         ]}}})))
   &SpriteRotationRate(Rotation Rate=&UniformRangedFloat(Minimum=-10.8, Maximum=10.8))
   &ScaleColor(
    Scale Alpha=&FloatFromCurve(FloatCurve=DataInterface() {propertyValues {Curve {keys [
         <at 0 0 RCIM_Cubic>
         {interpMode=RCIM_Cubic, tangentMode=RCTM_User, time=0.09, value=0.99}
         {arriveTangent=-1.0879, interpMode=RCIM_Cubic, leaveTangent=-1.0879, time=0.4777, value=0.63}
         <at 1 0 RCIM_Cubic>
        ]}}})
    Scale RGB=&VectorFromCurve(VectorCurve=DataInterface() {propertyValues {XCurve {keys [<at 0 1>,<at 0.85 1>]}, YCurve {keys [<at 0 1>,<at 0.85 1>]}, ZCurve {keys [<at 0 2>,<at 0.85 1>]}}})
    ScaleA=true
    ScaleRGB=true
   )
  ]
  rendererClasses [&NiagaraSpriteRendererProperties]
  renderers [{rendererClass=&NiagaraSpriteRendererProperties}]
 }
 Sparklies {
  simTarget=GPUComputeSim
  emitterUpdateScript [
   &EmitterState(Life Cycle Mode=::NewEnumerator1, Loop Behavior=::NewEnumerator1, Loop Duration=0.5, UseLoopDelay=false)
   &SpawnRate(SpawnRate=30, Use Spawn Probability=false)
   &SpawnBurst_Instantaneous(Spawn Count=25, Use Spawn Probability=false)
  ]
  particleSpawnScript [
   &InitializeParticle(
    Color=&UniformRangedLinearColor(Minimum=<rgba 0.0167 0.099 0.1 1>, Maximum=<rgba 0.1666 0.9901 1 1>)
    Lifetime=&UniformRangedFloat(Minimum=0.5, Maximum=0.8)
    Module.Write Color=true
    Module.Write Lifetime=true
    Module.Write Mass=true
    Module.Write Material Random=true
    Module.Write Position=false
    Module.Write Scale=false
    Module.Write SpriteRotation=true
    Module.Write SpriteSize=true
    Position=&SimulationPosition()
    Sprite Rotation=&NormalizedAngleToDegrees(NormalizedAngle=&UniformRangedFloat(Minimum=0, Maximum=1))
    Sprite Size=&Vector2DFromFloat(Value=&FloatFromCurve(FloatCurve=DataInterface() {propertyValues {Curve {keys [<at 0 150>,<at 1 50>]}}}, CurveIndex=$Emitter.NormalizedLoopAge))
   )
   &SkeletalMeshLocation(
    Inherit Velocity=true
    "Inherit Velocity (Scale)"=-0.2
    MaskSpawnGroup=false
    Module.Override Randomness=false
    Skeletal Mesh=$User.SkeletalMesh
    VelocitySpeedLimit=false
   )
   &SphereLocation(
    MaskSpawnGroup=false
    Override Seed=false
    Sphere Radius=&UniformRangedFloat(Minimum=0, Maximum=20)
    Surface Only=false
    UseNonUniformScale=false
    UseOffset=false
    UseSphereOrientation=false
   )
   &AddVelocity(Velocity=&Multiply_Vector(
     A=$Output.SphereLocation.SphereNormal
     B=&VectorFromFloat(Value=&UniformRangedFloat(Minimum=0, Maximum=1))
    ))
  ]
  particleUpdateScript [
   &ParticleState()
   &AccelerationForce(Acceleration=<0 0 -300>)
   &VectorNoiseForce(
    Force Amount=&UniformRangedFloat(Minimum=2500, Maximum=5000)
    Module.Use Random Noise Time Interval=true
    Random Noise Time Interval=0.15
   )
   &Drag(Drag=&UniformRangedFloat(Minimum=0, Maximum=0.5), Use Linear Drag=true, Use Rotational Drag=false)
   &SolveForcesAndVelocity(Clamp Velocity=false, Limit Acceleration=false)
   &ScaleColor(
    Scale Alpha=&FloatFromCurve(FloatCurve=DataInterface() {propertyValues {Curve {keys [<at 0.7 1 RCIM_Cubic>,<at 1 0>]}}})
    Scale RGB=&VectorFromCurve(VectorCurve=DataInterface() {propertyValues {
       XCurve {keys [<at 0.1 2 RCIM_Cubic>,<at 0.7 1 RCIM_Cubic>]}
       YCurve {keys [<at 0.1 2 RCIM_Cubic>,<at 0.7 0 RCIM_Cubic>]}
       ZCurve {keys [<at 0.1 2 RCIM_Cubic>,<at 0.7 1 RCIM_Cubic>]}
      }})
    ScaleA=true
    ScaleRGB=true
   )
   &SpriteRotationRate(Rotation Rate=&UniformRangedFloat(Minimum=-360, Maximum=360))
   &ScaleSpriteSize(Scale Factor=&Vector2DFromFloat(Value=&FloatFromCurve(FloatCurve=DataInterface() {propertyValues {Curve {keys [
          <at 0 0.49 RCIM_Cubic>
          {interpMode=RCIM_Cubic, tangentMode=RCTM_User, time=0.157, value=0.526}
          {arriveTangent=-5.4783, interpMode=RCIM_Cubic, leaveTangent=-5.4783, time=0.165, value=1.5}
          {interpMode=RCIM_Cubic, tangentMode=RCTM_User, time=0.18, value=0.4}
          {interpMode=RCIM_Cubic, tangentMode=RCTM_User, time=0.38, value=0.26}
          {arriveTangent=-0.2, interpMode=RCIM_Cubic, leaveTangent=-0.2, time=0.41, value=0.83}
          {interpMode=RCIM_Cubic, tangentMode=RCTM_User, time=0.43, value=0.25}
          {interpMode=RCIM_Cubic, tangentMode=RCTM_User, time=0.689, value=0.24}
          {arriveTangent=-0.7407, interpMode=RCIM_Cubic, leaveTangent=-0.7407, time=0.705, value=1}
          {interpMode=RCIM_Cubic, tangentMode=RCTM_User, time=0.716, value=0.22}
          <at 1 0 RCIM_Cubic>
         ]}}})))
  ]
  rendererClasses [&NiagaraSpriteRendererProperties]
  renderers [{rendererClass=&NiagaraSpriteRendererProperties}]
 }
 Sparklies_Burst {
  simTarget=GPUComputeSim
  emitterUpdateScript [
   &EmitterState(Life Cycle Mode=::NewEnumerator1, Loop Behavior=::NewEnumerator1, Loop Duration=0.5, UseLoopDelay=false)
   &SpawnBurst_Instantaneous(Spawn Count=12, Use Spawn Probability=false)
   &SpawnBurst_Instantaneous(Spawn Count=50, Spawn Time=0.02, Use Spawn Probability=false) {module=SpawnBurst_Instantaneous001}
   &SpawnBurst_Instantaneous(Spawn Count=37, Spawn Time=0.04, Use Spawn Probability=false) {module=SpawnBurst_Instantaneous002}
  ]
  particleSpawnScript [
   &InitializeParticle(
    Color=&UniformRangedLinearColor(Minimum=<rgba 0.0167 0.099 0.1 1>, Maximum=<rgba 0.1666 0.9901 1 1>)
    Lifetime=&UniformRangedFloat(Minimum=0.8, Maximum=1)
    Module.Write Color=true
    Module.Write Lifetime=true
    Module.Write Mass=true
    Module.Write Material Random=true
    Module.Write Position=false
    Module.Write Scale=false
    Module.Write SpriteRotation=true
    Module.Write SpriteSize=true
    Position=&SimulationPosition()
    Sprite Rotation=&NormalizedAngleToDegrees(NormalizedAngle=&UniformRangedFloat(Minimum=0, Maximum=1))
    Sprite Size=&Vector2DFromFloat(Value=&FloatFromCurve(FloatCurve=DataInterface() {propertyValues {Curve {keys [<at 0 150>,<at 1 50>]}}}, CurveIndex=$Emitter.NormalizedLoopAge))
   )
   &SkeletalMeshLocation(
    Inherit Velocity=true
    "Inherit Velocity (Scale)"=-1
    MaskSpawnGroup=false
    Module.Override Randomness=false
    Skeletal Mesh=$User.SkeletalMesh
    VelocitySpeedLimit=false
   )
   &SphereLocation(
    MaskSpawnGroup=false
    Override Seed=false
    Sphere Radius=&UniformRangedFloat(Minimum=0, Maximum=5)
    Surface Only=false
    UseNonUniformScale=false
    UseOffset=false
    UseSphereOrientation=false
   )
   &AddVelocity(Velocity=&Multiply_Vector(
     A=$Output.SphereLocation.SphereNormal
     B=&VectorFromFloat(Value=&FloatFromCurve(FloatCurve=DataInterface() {propertyValues {Curve {keys [
           {arriveTangent=-718.0126, interpMode=RCIM_Cubic, leaveTangent=-63.9312, tangentMode=RCTM_User, time=0, value=12}
           {arriveTangent=-0.53, interpMode=RCIM_Cubic, leaveTangent=-0.53, tangentMode=RCTM_User, time=0.4, value=1}
           <at 1 0 RCIM_Cubic>
          ]}}}))
    ))
   &AddVelocity(Velocity=&UniformRangedVector(Minimum=<0 0 30>, Maximum=<0 0 350>)) {module=AddVelocity001}
  ]
  particleUpdateScript [
   &ParticleState()
   &AccelerationForce(Acceleration=<0 0 -250>)
   &VectorNoiseForce(
    Force Amount=&Multiply_Float(
     A=&UniformRangedFloat(Minimum=2000, Maximum=4000)
     B=&FloatFromCurve(FloatCurve=DataInterface() {propertyValues {Curve {keys [<at 0 3 RCIM_Cubic>, {arriveTangent=-3, interpMode=RCIM_Cubic, leaveTangent=-3, time=0.27, value=1.5}, <at 1 0 RCIM_Cubic>]}}})
    )
    Module.Use Random Noise Time Interval=true
    Random Noise Time Interval=0.15
   )
   &Drag(Use Linear Drag=true, Use Rotational Drag=false)
   &SolveForcesAndVelocity(Clamp Velocity=false, Limit Acceleration=false)
   &ScaleColor(
    Scale Alpha=&FloatFromCurve(FloatCurve=DataInterface() {propertyValues {Curve {keys [<at 0.7 1 RCIM_Cubic>,<at 1 0>]}}})
    Scale RGB=&VectorFromCurve(VectorCurve=DataInterface() {propertyValues {
       XCurve {keys [<at 0.1 2 RCIM_Cubic>,<at 0.7 1 RCIM_Cubic>]}
       YCurve {keys [<at 0.1 2 RCIM_Cubic>,<at 0.7 0 RCIM_Cubic>]}
       ZCurve {keys [<at 0.1 2 RCIM_Cubic>,<at 0.7 1 RCIM_Cubic>]}
      }})
    ScaleA=true
    ScaleRGB=true
   )
   &SpriteRotationRate(Rotation Rate=&UniformRangedFloat(Minimum=-360, Maximum=360))
   &ScaleSpriteSize(Scale Factor=&Vector2DFromFloat(Value=&FloatFromCurve(FloatCurve=DataInterface() {propertyValues {Curve {keys [
          <at 0 0.49 RCIM_Cubic>
          {interpMode=RCIM_Cubic, tangentMode=RCTM_User, time=0.157, value=0.526}
          {arriveTangent=-5.4783, interpMode=RCIM_Cubic, leaveTangent=-5.4783, time=0.165, value=1.5}
          {interpMode=RCIM_Cubic, tangentMode=RCTM_User, time=0.18, value=0.4}
          {interpMode=RCIM_Cubic, tangentMode=RCTM_User, time=0.38, value=0.26}
          {arriveTangent=-0.2, interpMode=RCIM_Cubic, leaveTangent=-0.2, time=0.41, value=0.83}
          {interpMode=RCIM_Cubic, tangentMode=RCTM_User, time=0.43, value=0.25}
          {interpMode=RCIM_Cubic, tangentMode=RCTM_User, time=0.689, value=0.24}
          {arriveTangent=-0.7407, interpMode=RCIM_Cubic, leaveTangent=-0.7407, time=0.705, value=1}
          {interpMode=RCIM_Cubic, tangentMode=RCTM_User, time=0.716, value=0.22}
          <at 1 0 RCIM_Cubic>
         ]}}})))
  ]
  rendererClasses [&NiagaraSpriteRendererProperties]
  renderers [{rendererClass=&NiagaraSpriteRendererProperties}]
 }
 Sparklies_Tight_to_Body {
  simTarget=GPUComputeSim
  emitterUpdateScript [
   &EmitterState(Life Cycle Mode=::NewEnumerator1, Loop Behavior=::NewEnumerator1, Loop Duration=0.5, UseLoopDelay=false)
   &SpawnRate(
    SpawnRate=&FloatFromCurve(FloatCurve=DataInterface() {propertyValues {Curve {keys [<at 0 1>,<at 1 2>]}}}, CurveIndex=$Emitter.NormalizedLoopAge, Scale Curve=100)
    Use Spawn Probability=false
   )
  ]
  particleSpawnScript [
   &InitializeParticle(
    Color=&UniformRangedLinearColor(Minimum=<rgba 0.0167 0.099 0.1 1>, Maximum=<rgba 0.1666 0.9901 1 1>)
    Lifetime=&UniformRangedFloat(Minimum=1, Maximum=1.25)
    Module.Write Color=true
    Module.Write Lifetime=true
    Module.Write Mass=true
    Module.Write Material Random=true
    Module.Write Position=false
    Module.Write Scale=false
    Module.Write SpriteRotation=true
    Module.Write SpriteSize=true
    Position=&SimulationPosition()
    Sprite Rotation=&NormalizedAngleToDegrees(NormalizedAngle=&UniformRangedFloat(Minimum=0, Maximum=1))
    Sprite Size=&Vector2DFromFloat(Value=&FloatFromCurve(FloatCurve=DataInterface() {propertyValues {Curve {keys [<at 0 150>,<at 1 50>]}}}, CurveIndex=$Emitter.NormalizedLoopAge))
   )
   &SkeletalMeshLocation(
    Inherit Velocity=true
    "Inherit Velocity (Scale)"=-1
    MaskSpawnGroup=false
    Module.Override Randomness=false
    Skeletal Mesh=$User.SkeletalMesh
    VelocitySpeedLimit=false
   )
   &SphereLocation(
    MaskSpawnGroup=false
    Override Seed=false
    Sphere Radius=&UniformRangedFloat(Minimum=0, Maximum=20)
    Surface Only=false
    UseNonUniformScale=false
    UseOffset=false
    UseSphereOrientation=false
   )
   &AddVelocity(Velocity=&Multiply_Vector(
     A=$Output.SphereLocation.SphereNormal
     B=&VectorFromFloat(Value=&FloatFromCurve(FloatCurve=DataInterface() {propertyValues {Curve {keys [
           {arriveTangent=-376.3771, interpMode=RCIM_Cubic, leaveTangent=-376.3766, tangentMode=RCTM_User, time=0, value=22}
           {interpMode=RCIM_Cubic, tangentMode=RCTM_User, time=0.12, value=0.5}
           <at 1 0 RCIM_Cubic>
          ]}}}))
    ))
   &AddVelocity(Velocity=&UniformRangedVector(Minimum=<0 0 30>, Maximum=<0 0 350>)) {module=AddVelocity001}
  ]
  particleUpdateScript [
   &ParticleState()
   &AccelerationForce(Acceleration=<0 0 -250>)
   &VectorNoiseForce(
    Force Amount=&Multiply_Float(
     A=&UniformRangedFloat(Minimum=2000, Maximum=4000)
     B=&FloatFromCurve(FloatCurve=DataInterface() {propertyValues {Curve {keys [<at 0 3 RCIM_Cubic>, {arriveTangent=-3, interpMode=RCIM_Cubic, leaveTangent=-3, time=0.27, value=1.5}, <at 1 0 RCIM_Cubic>]}}})
    )
    Module.Use Random Noise Time Interval=true
    Random Noise Time Interval=0.15
   )
   &Drag(Drag=&Lerp_Float(A=1, B=2.5, Alpha=$Particles.NormalizedAge), Use Linear Drag=true, Use Rotational Drag=false)
   &SolveForcesAndVelocity(Clamp Velocity=false, Limit Acceleration=false)
   &ScaleColor(
    Scale Alpha=&FloatFromCurve(FloatCurve=DataInterface() {propertyValues {Curve {keys [<at 0.7 1 RCIM_Cubic>,<at 1 0>]}}})
    Scale RGB=&VectorFromCurve(VectorCurve=DataInterface() {propertyValues {
       XCurve {keys [<at 0.1 2 RCIM_Cubic>,<at 0.7 1 RCIM_Cubic>]}
       YCurve {keys [<at 0.1 2 RCIM_Cubic>,<at 0.7 0 RCIM_Cubic>]}
       ZCurve {keys [<at 0.1 2 RCIM_Cubic>,<at 0.7 1 RCIM_Cubic>]}
      }})
    ScaleA=true
    ScaleRGB=true
   )
   &SpriteRotationRate(Rotation Rate=&UniformRangedFloat(Minimum=-360, Maximum=360))
   &ScaleSpriteSize(Scale Factor=&Vector2DFromFloat(Value=&FloatFromCurve(FloatCurve=DataInterface() {propertyValues {Curve {keys [
          <at 0 0.49 RCIM_Cubic>
          {interpMode=RCIM_Cubic, tangentMode=RCTM_User, time=0.157, value=0.526}
          {arriveTangent=-5.4783, interpMode=RCIM_Cubic, leaveTangent=-5.4783, time=0.165, value=1.5}
          {interpMode=RCIM_Cubic, tangentMode=RCTM_User, time=0.18, value=0.4}
          {interpMode=RCIM_Cubic, tangentMode=RCTM_User, time=0.38, value=0.26}
          {arriveTangent=-0.2, interpMode=RCIM_Cubic, leaveTangent=-0.2, time=0.41, value=0.83}
          {interpMode=RCIM_Cubic, tangentMode=RCTM_User, time=0.43, value=0.25}
          {interpMode=RCIM_Cubic, tangentMode=RCTM_User, time=0.689, value=0.24}
          {arriveTangent=-0.7407, interpMode=RCIM_Cubic, leaveTangent=-0.7407, time=0.705, value=1}
          {interpMode=RCIM_Cubic, tangentMode=RCTM_User, time=0.716, value=0.22}
          <at 1 0 RCIM_Cubic>
         ]}}})))
  ]
  rendererClasses [&NiagaraSpriteRendererProperties]
  renderers [{rendererClass=&NiagaraSpriteRendererProperties}]
 }
}
summary {
 systemName=NS_Decoy_Teleport_Out_Athena
 emitters [
  {emitterName=Glow, rendererClasses [&NiagaraSpriteRendererProperties]}
  {emitterName=CoreFlare, rendererClasses [&NiagaraSpriteRendererProperties]}
  {emitterName=CoreFlare_MedHigh, rendererClasses [&NiagaraSpriteRendererProperties]}
  {emitterName=Corona, rendererClasses [&NiagaraSpriteRendererProperties]}
  {emitterName=Sparklies, simTarget=GPUComputeSim, rendererClasses [&NiagaraSpriteRendererProperties]}
  {emitterName=Fragments, rendererClasses [&NiagaraMeshRendererProperties]}
  {emitterName=Shimmer, rendererClasses [&NiagaraSpriteRendererProperties]}
  {emitterName=Sparklies_Tight_to_Body, simTarget=GPUComputeSim, rendererClasses [&NiagaraSpriteRendererProperties]}
  {emitterName=Sparklies_Burst, simTarget=GPUComputeSim, rendererClasses [&NiagaraSpriteRendererProperties]}
  {emitterName=Fragments_TightoBody, rendererClasses [&NiagaraMeshRendererProperties]}
  {emitterName=Fragments_Burst, rendererClasses [&NiagaraMeshRendererProperties]}
 ]
 userVariables [
  User.PointPosition {type=Vector3f, description="", defaultValue {struct=Vector3f, value=<0 0 250>}}
  User.SkeletalMesh {
   type=NiagaraDataInterfaceSkeletalMesh
   defaultValue {
    struct=NiagaraExt_VariableValue_DataInterface
    value {dataInterface=&self:NiagaraDataInterfaceSkeletalMesh_0}
   }
  }
 ]
}
