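# Niagara teleport-out effect for Athena player characters, dissolving the body into cyan-blue cubes, sprites and a shrinking core flare.
# Emitters sample User.SkeletalMesh; cubes and suction sprites use spawn groups gated by loop age to stagger three bone ranges.
# Suction sprites and square lines are pulled toward User.PointPosition.
# GPU cubes burst and collide, and delayed curl-noise Remnants spawn at their positions through an attribute reader.
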NiagaraSystem /Game/Effects/Fort_Effects/Effects/Characters/Generic/ResIn/NS_Player_Teleport_Out_Athena
dependencies=null
emitters {
 CoreFlare {
  emitterUpdateScript [
   &EmitterState(Life Cycle Mode=::NewEnumerator1, Loop Behavior=::NewEnumerator1, Loop Duration=0.5, UseLoopDelay=false)
   &SpawnBurst_Instantaneous(Spawn Count=1, Use Spawn Probability=false)
  ]
  particleSpawnScript [
   /Niagara/Modules/Spawn/Initialization/InitializeParticle(
    Color=<rgba 1.3467 2.6235 15 0.7>
    Module.Write Color=true
    Module.Write Lifetime=true
    Module.Write Mass=true
    Module.Write Material Random=true
    Module.Write Position=false
    Module.Write Scale=false
    Module.Write SpriteRotation=false
    Module.Write SpriteSize=true
    Position=&SimulationPosition()
    Sprite Size=<150 150>
   )
   &SkeletalMeshLocation(
    Inherit Velocity=false
    "Inherit Velocity (Scale)"=0
    MaskSpawnGroup=false
    Mesh Sampling Type=::NewEnumerator5
    Skeletal Mesh=$User.SkeletalMesh
    Socket Sampling Mode=::NewEnumerator5
   )
   &CameraOffset(Camera Offset Amount=50)
   &DynamicMaterialParameters(
    Index 0 Param 1=0.1
    Index 0 Param 2=1
    Index 0 Param 3=1
    Index 0 Param 4=1
    Module.Param0WriteEnabled=true
    Module.Param1WriteEnabled=true
    Module.Param2WriteEnabled=true
    Module.Param3WriteEnabled=true
   )
  ]
  particleUpdateScript [
   &ParticleState()
   &SolveForcesAndVelocity(Clamp Velocity=false, Limit Acceleration=false)
   &ScaleSpriteSize(Scale Factor=&Vector2DFromFloat(Value=&Lerp_Float(A=1, B=0, Alpha=$Particles.NormalizedAge)))
  ]
  rendererClasses [&NiagaraSpriteRendererProperties]
  renderers [{rendererClass=&NiagaraSpriteRendererProperties}]
 }
 Heavy_Cubes {
  simTarget=GPUComputeSim
  emitterUpdateScript [
   &EmitterState(Life Cycle Mode=::NewEnumerator1, Loop Behavior=::NewEnumerator1, Loop Duration=0.75, UseLoopDelay=false)
   &SpawnRate(SpawnRate=200, Use Spawn Probability=true)
   &SpawnRate(
    Spawn Probability=&MakeCustomFloatFromBool(Bool=&SetBoolByFloatComparison(A=$Emitter.NormalizedLoopAge, B=0.333), True Float=1, False Float=0)
    SpawnGroup=1
    SpawnRate=275
    Use Spawn Probability=true
   ) {module=SpawnRate001}
   &SpawnRate(
    Spawn Probability=&MakeCustomFloatFromBool(Bool=&SetBoolByFloatComparison(A=$Emitter.NormalizedLoopAge, B=0.666), True Float=1, False Float=0)
    SpawnGroup=2
    SpawnRate=200
    Use Spawn Probability=true
   ) {module=SpawnRate002}
  ]
  particleSpawnScript [
   &InitializeParticle(
    Color=<rgba 0.1912 0.5906 0.6038 1>
    Lifetime=&UniformRangedFloat(Minimum=0.2, Maximum=0.85)
    Mass=&UniformRangedFloat(Minimum=1, Maximum=2)
    Mesh Scale=&VectorFromFloat(Value=&UniformRangedFloat(Minimum=0.5, Maximum=2.5))
    Module.Write Color=true
    Module.Write Lifetime=true
    Module.Write Mass=true
    Module.Write Material Random=true
    Module.Write Position=false
    Module.Write Scale=true
    Module.Write SpriteRotation=false
    Module.Write SpriteSize=true
    Position=&SimulationPosition()
    Sprite Size=<50 50>
   )
   &CameraOffset(Camera Offset Amount=5)
   &SkeletalMeshLocation(
    Bone / Socket Index=&self:BoneIndexRangebySpawnGroup(
     MaxIndex_SpawnGroup0:NiagaraInt32=2
     MaxIndex_SpawnGroup1:NiagaraInt32=11
     MaxIndex_SpawnGroup2:NiagaraInt32=18
     MinIndex_SpawnGroup0:NiagaraInt32=0
     MinIndex_SpawnGroup1:NiagaraInt32=3
     MinIndex_SpawnGroup2:NiagaraInt32=12
     Spawn Group:NiagaraInt32=&SpawnGroup()
    )
    Bone Sampling Mode=::NewEnumerator4
    Inherit Velocity=true
    Interpolate Mesh Sampling=false
    MaskSpawnGroup=false
    Skeletal Mesh=$User.SkeletalMesh
    VelocitySpeedLimit=false
   )
   &SphereLocation(
    MaskSpawnGroup=false
    Override Seed=false
    Sphere Radius=20
    Surface Only=false
    UseNonUniformScale=false
    UseOffset=false
    UseSphereOrientation=false
   )
   &MeshRotationForce(Rotation=<1 1 1>)
   &PointForce(Force Strength=500, Module.Use Falloff=false, Module.UseLinearForce=false)
   &ApplyInitialForces()
   &InitialMeshOrientation(
    Rotation=&UniformRangedVector(Minimum=<0 0 0>, Maximum=<1 1 1>)
    Use Orientation Vector=false
    Use Rotation Vector=true
   ) {module=InitialMeshOrientation001}
  ]
  particleUpdateScript [
   &ParticleState()
   &GravityForce()
   &Drag(Drag=&UniformRangedFloat(Minimum=4, Maximum=8), Use Linear Drag=true, Use Rotational Drag=false)
   &Collision(
    Age Colliding Particles=false
    Control Roll On Collision=true
    GPU Collision Type=::NewEnumerator2
    Module.Control Rest State Via Surface Normal=false
    Module.EnableMaxCollisionCount=false
    Module.Randomize Collision Normal Vector=false
    Radius Calculation Type=::NewEnumerator1
    ScaleRateOfRotation=false
   )
   &SolveForcesAndVelocity(Clamp Velocity=false, Limit Acceleration=false)
   &ScaleColor(
    Scale Alpha=&FloatFromCurve(FloatCurve=DataInterface() {propertyValues {Curve {keys [<at 0.5 1 RCIM_Cubic>, {arriveTangent=-4.6302, interpMode=RCIM_Cubic, leaveTangent=-4.6302, tangentMode=RCTM_User, time=1, value=0}]}}})
    Scale RGB=&VectorFromFloat(Value=&FloatFromCurve(FloatCurve=DataInterface() {propertyValues {Curve {keys [
          {arriveTangent=32.9526, interpMode=RCIM_Cubic, time=0, value=10}
          <at 0.1 1>
          {interpMode=RCIM_Cubic, leaveTangent=-29.5004, time=1, value=0}
         ]}}}))
    ScaleA=true
    ScaleRGB=true
   )
   &ScaleMeshSize(Scale Factor=&VectorFromFloat(Value=&FloatFromCurve(FloatCurve=DataInterface() {propertyValues {Curve {keys [<at 0.75 1 RCIM_Cubic>, {arriveTangent=-4.0622, interpMode=RCIM_Cubic, leaveTangent=-4.0622, tangentMode=RCTM_User, time=1, value=0.25}]}}}))) {enabled=false}
  ]
  rendererClasses [&NiagaraMeshRendererProperties]
  renderers [{rendererClass=&NiagaraMeshRendererProperties, rendererIndex=0}]
 }
 Remnants {
  simTarget=GPUComputeSim
  emitterUpdateScript [
   &EmitterState(Life Cycle Mode=::NewEnumerator1, Loop Behavior=::NewEnumerator1, Loop Delay=0.6, Loop Duration=0.35, UseLoopDelay=true)
   &SpawnRate(SpawnRate=375, Use Spawn Probability=false)
  ]
  particleSpawnScript [
   &InitializeParticle(
    Color=&MakeLinearColorFromVectorAndFloat(
     "Vector (RGB)"=&MakeVectorFromLinearColorRGB(Color=<rgba 0.6333 1.9563 2 1>)
     "Float (Alpha)"=&UniformRangedFloat(Minimum=0.25, Maximum=0.5)
    )
    Lifetime=&UniformRangedFloat(Minimum=0.55, Maximum=1.05)
    Module.Write Color=true
    Module.Write Lifetime=true
    Module.Write Mass=true
    Module.Write Material Random=true
    Module.Write Position=false
    Module.Write Scale=false
    Module.Write SpriteRotation=false
    Module.Write SpriteSize=true
    Position=&SimulationPosition()
    Sprite Size=&UniformRangedVector2D(Minimum=<5 5>, Maximum=<25 25>)
   )
   /Game/Effects/Niagara/Modules/Attributes/SetLocationFromEmitter(AttributeReader=DataInterface() {propertyValues {EmitterBinding {bindingMode=Other, emitterName=Heavy_Cubes}}})
   &DynamicMaterialParameters(
    Index 0 Param 1=&UniformRangedFloat(Minimum=0.1, Maximum=0.3)
    Index 0 Param 2=&UniformRangedFloat(Minimum=0.75, Maximum=1.25)
    Index 0 Param 3=&UniformRangedFloat(Minimum=0.2, Maximum=0.7)
    Module.Param0WriteEnabled=true
    Module.Param1WriteEnabled=true
    Module.Param2WriteEnabled=true
    Module.Param3WriteEnabled=false
   )
  ]
  particleUpdateScript [
   &ParticleState()
   /Niagara/Modules/Update/Forces/CurlNoiseForce(Noise Frequency=0.025, Noise Strength=500, Pan Noise=true, Pan Noise Field=<0 0 2>)
   &SolveForcesAndVelocity(Clamp Velocity=false, Limit Acceleration=false)
   &ScaleColor(
    Scale Alpha=&FloatFromCurve(FloatCurve=DataInterface() {propertyValues {Curve {keys [{arriveTangent=-3.2179, interpMode=RCIM_Cubic, leaveTangent=-3.2179, tangentMode=RCTM_User, time=0, value=1}, <at 1 0 RCIM_Cubic>]}}})
    ScaleA=true
    ScaleRGB=true
   )
   &ScaleSpriteSize(Scale Factor=&Vector2DFromFloat(Value=&FloatFromCurve(FloatCurve=DataInterface() {propertyValues {Curve {keys [<at 0 1 RCIM_Cubic>,<at 1 0 RCIM_Cubic>]}}})))
   &DynamicMaterialParameters(
    Index 0 Param 4=&FloatFromCurve(FloatCurve=DataInterface() {propertyValues {Curve {keys [{arriveTangent=2.0856, interpMode=RCIM_Cubic, leaveTangent=2.0856, tangentMode=RCTM_User, time=0, value=0}, <at 1 1 RCIM_Cubic>]}}})
    Module.Param0WriteEnabled=false
    Module.Param1WriteEnabled=false
    Module.Param2WriteEnabled=false
    Module.Param3WriteEnabled=true
   ) {module=DynamicMaterialParameters001}
  ]
  rendererClasses [&NiagaraSpriteRendererProperties]
  renderers [{rendererClass=&NiagaraSpriteRendererProperties}]
 }
 Sprite_Suction {
  emitterUpdateScript [
   &EmitterState(Life Cycle Mode=::NewEnumerator1, Loop Behavior=::NewEnumerator1, Loop Duration=0.65, UseLoopDelay=false)
   &SpawnRate(SpawnRate=100, Use Spawn Probability=true)
   &SpawnRate(
    Spawn Probability=&MakeCustomFloatFromBool(Bool=&SetBoolByFloatComparison(A=$Emitter.NormalizedLoopAge, B=0.333), True Float=1, False Float=0)
    SpawnGroup=1
    SpawnRate=100
    Use Spawn Probability=true
   ) {module=SpawnRate001}
   &SpawnRate(
    Spawn Probability=&MakeCustomFloatFromBool(Bool=&SetBoolByFloatComparison(A=$Emitter.NormalizedLoopAge, B=0.666), True Float=1, False Float=0)
    SpawnGroup=2
    SpawnRate=100
    Use Spawn Probability=true
   ) {module=SpawnRate002}
  ]
  particleSpawnScript [
   &InitializeParticle(
    Color=&UniformRangedLinearColor(Minimum=<rgba 0.1912 0.5906 0.6038 0.25>, Maximum=<rgba 0.3958 1.2227 1.25 1>)
    Lifetime=&UniformRangedFloat(Minimum=0.3, Maximum=0.7)
    Mass=&UniformRangedFloat(Minimum=0.25, Maximum=2)
    Module.Write Color=true
    Module.Write Lifetime=true
    Module.Write Mass=true
    Module.Write Material Random=true
    Module.Write Position=false
    Module.Write Scale=false
    Module.Write SpriteRotation=false
    Module.Write SpriteSize=true
    Position=&SimulationPosition()
    Sprite Size=&Vector2DFromFloat(Value=&UniformRangedFloat(Minimum=5, Maximum=7))
   )
   &CameraOffset(Camera Offset Amount=10)
   &SkeletalMeshLocation(
    Bone / Socket Index=&self:BoneIndexRangebySpawnGroup_01(
     MaxIndex_SpawnGroup0:NiagaraInt32=2
     MaxIndex_SpawnGroup1:NiagaraInt32=11
     MaxIndex_SpawnGroup2:NiagaraInt32=18
     MinIndex_SpawnGroup0:NiagaraInt32=0
     MinIndex_SpawnGroup1:NiagaraInt32=3
     MinIndex_SpawnGroup2:NiagaraInt32=12
     Spawn Group:NiagaraInt32=&SpawnGroup()
    )
    Bone Sampling Mode=::NewEnumerator4
    Inherit Velocity=true
    Interpolate Mesh Sampling=false
    MaskSpawnGroup=false
    Skeletal Mesh=$User.SkeletalMesh
    VelocitySpeedLimit=false
   )
   &SphereLocation(
    MaskSpawnGroup=false
    Override Seed=false
    Sphere Radius=20
    Surface Only=false
    UseNonUniformScale=false
    UseOffset=false
    UseSphereOrientation=false
   )
   &InheritVelocity(Module.Limit Speed=true)
   &PointForce(
    Force Origin=$Output.SkeletalMeshLocation.SampledPosition
    Force Strength=&UniformRangedFloat(Minimum=150, Maximum=375)
    Module.Use Falloff=false
    Module.UseLinearForce=false
   )
   &ApplyInitialForces()
   &DynamicMaterialParameters(
    Index 0 Param 1=&UniformRangedFloat(Minimum=0, Maximum=1)
    Index 0 Param 2=&UniformRangedFloat(Minimum=5, Maximum=25)
    Index 0 Param 3=1
    Module.Param0WriteEnabled=true
    Module.Param1WriteEnabled=true
    Module.Param2WriteEnabled=true
    Module.Param3WriteEnabled=true
   )
  ]
  particleUpdateScript [
   &ParticleState()
   &CurlNoiseForce(Noise Frequency=0.025, Noise Strength=1000, Pan Noise=true, Pan Noise Field=<0 0 2>)
   &GravityForce(Gravity=<0 0 -500>)
   &Drag(Drag=&UniformRangedFloat(Minimum=3, Maximum=5), Use Linear Drag=true, Use Rotational Drag=false)
   /Niagara/Modules/Update/Forces/PointAttractionForce(
    Attraction Radius=1000
    AttractionStrength:NiagaraFloat=&FloatFromCurve(FloatCurve=DataInterface() {propertyValues {Curve {keys [<at 0.25 0 RCIM_Cubic>,<at 0.7 1 RCIM_Cubic>]}}}, Scale Curve=350)
    AttractorPosition:Vector3f=$User.PointPosition
    Module.Kill Within Radius=false
    Module.Use Falloff=true
   )
   &SolveForcesAndVelocity(Clamp Velocity=false, Limit Acceleration=false) {enabled=false}
   &ScaleColor(
    Scale RGB=&VectorFromFloat(Value=&FloatFromCurve(FloatCurve=DataInterface() {propertyValues {Curve {keys [<at 0 5>,<at 1 0>]}}}))
    ScaleA=true
    ScaleRGB=true
   ) {enabled=false}
   &SolveForcesAndVelocity(Clamp Velocity=false, Limit Acceleration=false) {module=SolveForcesAndVelocity001}
   &ScaleSpriteSizeBySpeed(
    Max Scale Factor=<0.1 3>
    Sample Scale Factor By Curve=true
    Scale Factor Curve=DataInterface() {propertyValues {Curve {keys [<at 0 0 RCIM_Cubic>, {arriveTangent=3.2239, interpMode=RCIM_Cubic, leaveTangent=3.2239, tangentMode=RCTM_User, time=1, value=1}]}}}
   )
   &DynamicMaterialParameters(
    Index 0 Param 3=&FloatFromCurve(
     FloatCurve=DataInterface() {propertyValues {Curve {keys [{arriveTangent=-2.6201, interpMode=RCIM_Cubic, leaveTangent=-2.6201, tangentMode=RCTM_User, time=0, value=1}, <at 1 0 RCIM_Cubic>]}}}
     Scale Curve=10
    )
    Index 0 Param 4=&FloatFromCurve(FloatCurve=DataInterface() {propertyValues {Curve {keys [{arriveTangent=1.1788, interpMode=RCIM_Cubic, leaveTangent=1.1788, tangentMode=RCTM_User, time=0, value=0}, <at 1 1 RCIM_Cubic>]}}})
    Module.Param0WriteEnabled=false
    Module.Param1WriteEnabled=false
    Module.Param2WriteEnabled=true
    Module.Param3WriteEnabled=true
   ) {module=DynamicMaterialParameters001}
  ]
  rendererClasses [&NiagaraSpriteRendererProperties]
  renderers [{rendererClass=&NiagaraSpriteRendererProperties}]
 }
 SquareLines {
  emitterUpdateScript [
   &EmitterState(Life Cycle Mode=::NewEnumerator1, Loop Behavior=::NewEnumerator1, Loop Duration=0.9, UseLoopDelay=false)
   &SpawnRate(SpawnRate=75, Use Spawn Probability=false)
  ]
  particleSpawnScript [
   &InitializeParticle(
    Color=<rgba 0.1494 0.8409 3 1>
    Lifetime=&UniformRangedFloat(Minimum=0.25, Maximum=0.7)
    Module.Write Color=true
    Module.Write Lifetime=true
    Module.Write Mass=true
    Module.Write Material Random=true
    Module.Write Position=false
    Module.Write Scale=false
    Module.Write SpriteRotation=false
    Module.Write SpriteSize=true
    Position=&SimulationPosition()
    Sprite Size=&UniformRangedVector2D(Minimum=<2 6>, Maximum=<3 10>)
   )
   &SkeletalMeshLocation(Inherit Velocity=false, MaskSpawnGroup=false, Module.Override Randomness=false, Skeletal Mesh=$User.SkeletalMesh)
   &SphereLocation(
    MaskSpawnGroup=false
    Override Seed=false
    Sphere Radius=5
    Surface Only=false
    UseNonUniformScale=false
    UseOffset=false
    UseSphereOrientation=false
   )
   &CameraOffset(Camera Offset Amount=50)
   &AddVelocity(Velocity=&Multiply_Vector(
     A=$Output.SphereLocation.SphereNormal
     B=&VectorFromFloat(Value=&UniformRangedFloat(Minimum=1, Maximum=20))
    ))
  ]
  particleUpdateScript [
   &ParticleState()
   &AccelerationForce(Acceleration=<0 0 300>)
   &Drag(Drag=&UniformRangedFloat(Minimum=0, Maximum=2), Use Linear Drag=true, Use Rotational Drag=false)
   &PointAttractionForce(
    Attraction Radius=5000
    AttractionStrength=150
    AttractorPosition:Vector3f=$User.PointPosition
    Kill Radius=10
    Module.Kill Within Radius=true
    Module.Use Falloff=true
   )
   &SolveForcesAndVelocity(Clamp Velocity=false, Limit Acceleration=false)
   &ScaleSpriteSize(Scale Factor=&Vector2DFromFloat(Value=&FloatFromCurve(FloatCurve=DataInterface() {propertyValues {Curve {keys [<at 0 1 RCIM_Cubic>, {arriveTangent=-3.0707, interpMode=RCIM_Cubic, leaveTangent=-3.0707, tangentMode=RCTM_User, time=1, value=0}]}}})))
  ]
  rendererClasses [&NiagaraSpriteRendererProperties]
  renderers [{rendererClass=&NiagaraSpriteRendererProperties}]
 }
}
summary {
 systemName=NS_Player_Teleport_Out_Athena
 emitters [
  {emitterName=CoreFlare, rendererClasses [&NiagaraSpriteRendererProperties]}
  {bEnabled=false, emitterName=Squares, rendererClasses [&NiagaraSpriteRendererProperties]}
  {emitterName=SquareLines, rendererClasses [&NiagaraSpriteRendererProperties]}
  {emitterName=Heavy_Cubes, simTarget=GPUComputeSim, rendererClasses [&NiagaraMeshRendererProperties]}
  {emitterName=Sprite_Suction, rendererClasses [&NiagaraSpriteRendererProperties]}
  {emitterName=Remnants, simTarget=GPUComputeSim, rendererClasses [&NiagaraSpriteRendererProperties]}
 ]
 userVariables [
  User.PointPosition {type=Vector3f, description="", defaultValue {struct=Vector3f, value=<0 0 0>}}
  User.SkeletalMesh {
   type=NiagaraDataInterfaceSkeletalMesh
   defaultValue {
    struct=NiagaraExt_VariableValue_DataInterface
    value {dataInterface=&self:NiagaraDataInterfaceSkeletalMesh_1}
   }
  }
 ]
}
