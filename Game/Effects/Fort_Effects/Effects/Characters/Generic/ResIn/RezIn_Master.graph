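# Translucent unlit master material for the character and weapon "rez in" effect, drawing an animated glowing swipe band.
# CharacterResInEffect's gradient, contrast-adjusted, samples ShieldPowerAnim to drive color and opacity, mixed with grid and world-aligned sparkle noise.
# The band also masks a vertex push along mesh normals or gun-function smooth normals.
# A Box static switch drops noise and push, masking opacity with view-angle falloff, depth fade and cylinder or square gradient.
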Material /Game/Effects/Fort_Effects/Effects/Characters/Generic/ResIn/RezIn_Master
graph {
 properties {
  bEnableResponsiveAA=true
  bUsedWithClothing=true
  bUsedWithMeshDeformer=true
  bUsedWithMorphTargets=true
  bUsedWithNanite=true
  bUsedWithSkeletalMesh=true
  blendMode=BLEND_Translucent
  pixelDepthOffsetMode=PDOM_Legacy
  shadingModel=MSM_Unlit
  usedShadingModels=Unlit
 }
 nodes [
  n1 = VectorParameter() {parameterName=WireFrameFadeOffColor, defaultValue=<rgba 0.2654 0.411 0.98 2>}
  n2 = MaterialFunctionCall(
   In=MaterialFunctionCall!(Animation Gradient, Rez In Percentage) {materialFunction=/Game/Packages/Fortress_SharedMaterials/Base_Material_Functions/CharacterResInEffect}.Animation With Distortion
   Contrast=ScalarParameter() {defaultValue=1, parameterName=Swipe Band Width}
  ) {materialFunction=&CheapContrast}
  n3 = TextureSample(UVs=@n2.Result) {samplerType=SAMPLERTYPE_LinearColor, texture=/Game/Packages/Fortnite_UniqueMaterials_1/Effects/ShieldPowerAnim}
  n4 = Multiply(
   A=ScalarParameter() {defaultValue=0.25, parameterName=Scale}
   B=MaterialFunctionCall!(Center Point World Position=ObjectPositionWS()) {materialFunction=/Game/Packages/Fortnite_UniqueMaterials_1/Effects/WorldAlignedScreenCoordinates}.Result
  )
  n5 = TextureSample(UVs=Multiply(A=@n4) {constB=1.1}) {samplerType=SAMPLERTYPE_LinearColor, texture=/Game/Gadgets/Assets/Slow_Field/Textures/T_Slowfield_SparkleNoise_01}
  n6 = TextureSample(UVs=Panner(Coordinate=@n4) {speedX=0.1, speedY=0.1}) {samplerType=SAMPLERTYPE_LinearColor, texture=&T_Slowfield_SparkleNoise_01}
  n7 = MaterialFunctionCall!(
   Base=Multiply(A=Multiply(A=@n3.R, B=@n3.R) {constB=2}, B=ScalarParameter() {defaultValue=0.75, parameterName=BandInfluenceMultiplier}) {constB=2}
   Add=Multiply(A=Multiply(A=@n5.R, B=@n6.G)) {constB=10}
  ) {materialFunction=/Engine/Functions/Engine_MaterialFunctions02/Math/MultiplyAdd}
  n8 = Power(Base=@n2.Result, Exp=ScalarParameter() {defaultValue=2, parameterName=Gradient Shrinkage})
  n9 = TextureSample(UVs=@n8) {samplerType=SAMPLERTYPE_LinearColor, texture=&ShieldPowerAnim}
  n10 = Power(Base=@n9.R, Exp=ScalarParameter() {defaultValue=1, parameterName=Band Power})
  n11 = LinearInterpolate(
   A=@n1.RGB
   B=VectorParameter() {parameterName=WireFrameParameterHighlight, defaultValue=<rgba 0.7175 1.0246 1.8 1>}.RGB
   Alpha=StaticSwitchParameter(True=@n10, False=Add(A=@n10, B=@n7.Result)) {parameterName=Box}
  )
  n12 = VectorParameter() {parameterName=Bottom Color, defaultValue=<rgba 0.0284 0.0348 0.05 1>}
  n13 = VectorParameter() {parameterName=Top Color, defaultValue=<rgba 0.6534 0.8914 1.8 1>}
  n14 = MaterialFunctionCall() {materialFunction=&BoundingBoxBased_0-1_UVW}
  n15 = MaterialFunctionCall() {materialFunction=&CharacterResInEffect}
  Multiply(
   A=MaterialFunctionCall(Base=Power(Base=@n11, Exp=@n1.A) {desc=emissive}, Add=LinearInterpolate(A=@n12.RGB, B=@n13.RGB, Alpha=@n14.B)) {materialFunction=&MultiplyAdd}.Result
   B=Multiply(A=LinearInterpolate(Alpha=@n15.FirstSpawnPassComplete) {constA=0.5}, B=ScalarParameter() {defaultValue=1, parameterName=Emissive Modulation})
  ) {desc=E}
  n16 = TextureSample(UVs=TextureCoordinate()) {samplerType=SAMPLERTYPE_LinearColor, texture=/Game/Global/Textures/Default/Noise/T_Noise01}
  n17 = ScalarParameter() {defaultValue=6.7645, parameterName=GridTiling}
  n18 = TextureCoordinate()
  n19 = Multiply(
   A=TextureSample(UVs=Multiply(A=@n17, B=@n18)) {texture=&/DiffusedStylizedGrid}.B
   B=ScalarParameter() {defaultValue=3.8285, parameterName=GridOpacity}
  ) {constB=0.1}
  n20 = Multiply(A=@n10, B=StaticSwitchParameter(True=Constant() {r=1}, False=Multiply(A=@n16.R, B=@n19)) {parameterName=Box})
  n21 = Clamp(None=StaticSwitchParameter(True=@n10, False=Add(A=@n20, B=@n7.Result) {constB=0}) {parameterName=Box})
  n22 = Multiply(A=@n21, B=@n21)
  n23 = DotProduct(A=CameraVectorWS(), B=PixelNormalWS())
  n24 = ScalarParameter() {defaultValue=0.223, parameterName=Fresnel Power}
  n25 = StaticSwitchParameter(
   True=TextureSample() {texture=/Game/Global/Textures/Default/Masks_and_Gradients/CylGradient}.R
   False=TextureSample() {texture=&/SquareGradient}.R
  ) {parameterName=Cylinder}
  n26 = ScalarParameter() {defaultValue=0.2, parameterName=Outer Glow Ring Modification}
  n27 = Multiply(A=Multiply(A=Power(Base=@n23, Exp=@n24), B=Multiply(A=@n25, B=@n26)), B=DepthFade() {fadeDistanceDefault=10})
  Multiply(
   A=StaticSwitchParameter(True=Multiply(A=@n22, B=@n27), False=@n22) {desc=opacity, group=Box, parameterName=Box}
   B=ScalarParameter() {defaultValue=1, parameterName=Opacity Modulation}
  ) {desc=O}
  n28 = MaterialFunctionCall!(Use World Position Offsets) {materialFunction=/Engine/Functions/Engine_MaterialFunctions02/Utility/CameraVectorWithWPOOptions}
  n29 = ShadingPathSwitch(Default=@n9.A, Deferred=@n9.A, Mobile=@n9.A) {desc=Migrated from FeatureLevelSwitch}
  n30 = Multiply(
   A=Power(Base=OneMinus(None=@n29) {desc=vertex shader}, Exp=ScalarParameter() {defaultValue=10, parameterName=Push Mask Power}) {constExponent=10}
   B=ScalarParameter() {defaultValue=1.5, parameterName=Max Push Amount}
  )
  n31 = LinearInterpolate(
   A=MaterialFunctionCall!(UV 1 coordinates, MuzzleFalloff) {materialFunction=/Game/Packages/Fortnite_UniqueMaterials_1/Weapons/M_GunMasterShaderFunction}.Smooth Normal
   B=VertexNormalWS()
   Alpha=ScalarParameter() {parameterName=Character 1 Gun 0}
  )
  StaticSwitchParameter(True=@n28.Result, False=Add(A=@n28.Result, B=Multiply(A=@n30, B=@n31) {constA=10})) {desc=opacity, group=Box, parameterName=Box}
  TextureSample!(UVs=@n8, Tex, Level=Constant() {r=3}, Apply View MipBias) {mipValueMode=TMVM_MipLevel, samplerType=SAMPLERTYPE_LinearColor, texture=&ShieldPowerAnim}
 ]
}
